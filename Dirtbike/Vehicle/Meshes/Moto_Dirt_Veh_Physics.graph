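# Physics asset for the dirt bike vehicle mesh, defining its collision as one compound body on the root bone.
# Named boxes approximate the front fender, forks, swingarms and rear seat, while a capsule covers the handlebars.
# Two larger capsules stand in for the driver and passenger, so rider volumes are part of the bike's own collision.
# It has no constraints and no separate bodies for wheels or other moving parts.
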PhysicsAsset /Dirtbike/Vehicle/Meshes/Moto_Dirt_Veh_Physics
bodies {root [
  {shapeType=Capsule, length=60.8125, radius=5.875, shapeName=HandleBars, center=<19.475 0.0001 98.5543>, rotation=<rot 0 0 -90.0001>}
  {shapeType=Capsule, length=46, radius=23.4375, shapeName=DriverCollision, center=<-13.1506 0 109.1664>}
  {shapeType=Capsule, length=62.9375, radius=23.4375, shapeName=PassengerCollision, center=<-68.44 0 135.9241>}
  {shapeType=Box, extentX=86, extentY=36.875, extentZ=24, shapeName=FrontFender, center=<57.3009 0.0002 68.2325>, rotation=<rot 16.875 0 0>}
  {
   shapeType=Box
   extentX=50.625
   extentY=5.1875
   extentZ=9.375
   shapeName=RightSwingarm
   center=<-38.6426 11.8292 30.1207>
   rotation=<rot 15.384 -0.0002 -0.0007>
  }
  {
   shapeType=Box
   extentX=50.625
   extentY=5.1875
   extentZ=9.375
   shapeName=LeftSwingarm
   center=<-38.6426 -11.5269 30.1205>
   rotation=<rot 15.384 -0.0002 -0.0007>
  }
  {
   shapeType=Box
   extentX=33.875
   extentY=7.5
   extentZ=9.4375
   shapeName=FrontFender2
   center=<36.6095 -0.0001 64.3338>
   rotation=<rot 51.8894 -0.0012 -0.0015>
  }
  {
   shapeType=Box
   extentX=37.8125
   extentY=7.5
   extentZ=9.4375
   shapeName=FrontFender1
   center=<61.8388 -0.0003 74.0365>
   rotation=<rot -5.3936 0.0001 -0.0013>
  }
  {
   shapeType=Box
   extentX=6.75
   extentY=7.5
   extentZ=65.75
   shapeName=FrontForkRight
   center=<49.8262 -6.7032 55.3123>
   rotation=<rot 29.894 -0.0014 -0.0029>
  }
  {
   shapeType=Box
   extentX=6.75
   extentY=7.5
   extentZ=65.75
   shapeName=FrontForkLeft
   center=<49.8262 8.1492 55.3129>
   rotation=<rot 29.894 -0.0014 -0.0029>
  }
  {shapeType=Box, extentX=27.8125, extentY=7.5, extentZ=9.0625, shapeName=RearSeat, center=<-78.97 0 78.0929>}
 ]}
constraints []
