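# Physics asset for the dirt bike vehicle, named for fluid simulation use, giving a coarse collision approximation of the bike.
# Seven bones each carry a single oriented box: body, frame, front differential, muffler, front steering assembly, and front and back tilted wheels.
# Wheels use boxes rather than rounded shapes, and the frame box is only about one unit across.
# No constraints join the bodies.
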PhysicsAsset /Dirtbike/Vehicle/Meshes/Moto_Dirt_Veh_Physics_Fluidsim
bodies {
 body [{shapeType=Box, extentX=13.7883, extentY=38.8746, extentZ=138.058, center=<-25.31 -0.0736 0.9332>, rotation=<rot -0.0001 -90.125 -79.4558>}]
 frame [{shapeType=Box, extentX=1.01, extentY=1.01, extentZ=1.01, rotation=<rot 0 -90 0>}]
 front_differential [{shapeType=Box, extentX=26.9632, extentY=15.9734, extentZ=35.7945, center=<-10.3876 -1.0594 -2.2947>, rotation=<rot 0 -94.7859 -78.3895>}]
 muffler [{shapeType=Box, extentX=12.3859, extentY=12.5876, extentZ=48.7451, center=<24.0335 0 -0.4449>, rotation=<rot 0 90 -88.3523>}]
 wheel_steering_fr [{shapeType=Box, extentX=13.7195, extentY=47.4024, extentZ=73.1676, center=<13.2297 -6.4983 -26.6846>, rotation=<rot 0 78.723 0.4131>}]
 wheel_tilted_bk [{shapeType=Box, extentX=18.5787, extentY=39.1128, extentZ=39.1165, rotation=<rot -1.3146 -90 43.8425>}]
 wheel_tilted_fr [{shapeType=Box, extentX=13.6697, extentY=48.3406, extentZ=48.3943, center=<0 0.0001 0>, rotation=<rot -0.7964 -90 -17.6853>}]
}
constraints []
