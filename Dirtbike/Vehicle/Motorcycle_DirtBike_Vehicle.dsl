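; Dirt bike vehicle Blueprint for Fortnite, handling riders, damage, explosion, surface and water reactions.
; BeginPlay binds seat, health, gear, surface and destruction delegates, whose handlers lean on MotorcycleFunctionLibrary, gameplay effects and Niagara parameters.
; Critical health can ignite Curie fire; explosions split server damage from client cosmetics, with exported particle data spawning mud splashes.
; The audio controller loads asynchronously, and a looping timer drives water foam, bubbles and tire spray by depth.
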(event EventOnHoldExitStarted
  (params (ExitingPawn :type Fort%20Pawn%20Object%20Reference) (ExitDuration :type Float%20%28single-precision%29))
  ;[1 1 1] EVENT -> OnHoldExitStarted/Stopped
  (Ability|SendGameplayEventtoActor ExitingPawn "(TagName=\"Athena.Vehicle.Biplane.OnHoldExitStarted\")" (Utilities|Struct|MakeGameplayEventData :ContextHandle "()" :InstigatorTags "(GameplayTags=)" :TargetTags "(GameplayTags=)" :EventMagnitude ExitDuration :TargetData "()")))

(event EventOnHoldExitStopped
  (params (ExitingPawn :type Fort%20Pawn%20Object%20Reference))
  ;[1 1 1] EVENT -> OnHoldExitStarted/Stopped
  (Ability|SendGameplayEventtoActor ExitingPawn "(TagName=\"Athena.Vehicle.Biplane.OnHoldExitStopped\")"))

(event Custom|HandleOnStartCriticalHealth
  (params (LastDamageInstigator :type Controller%20Object%20Reference))
  ;[1 1 1] EVENT -> Start Critical Health
  (Utilities|FlowControl|DoOnce false
    (:Completed
      (if (> (ScalableFloat|GetValueatLevel (Variables|Curie|FX|GetHotfixCurieEnabled)) 0.0)
        (Curie|OnCurieActivated
          (:OnCurieActive
            (Curie|EnableElementAttachmentonActor self "(TagName=\"Curie.Element.Fire\")")
            (Curie|ExecuteFireApplication self self self 1e+02 (Collision|MakeHitResult false false 0.0 0.0 (Transformation|GetActorLocation) (Transformation|GetActorLocation) "0,0,1" "0,0,1" 0 self (Variables|Vehicle|GetSkeletalMesh)) -1 128.0)))))))

(event Custom|Event_OnDamageablePartHealthChanged
  (params (PartConfig :type Vehicle%20Damageable%20Part%20Config%20Structure%20%28by%20ref%29) (Part :type Vehicle%20Damageable%20Part%20Structure%20%28by%20ref%29) (bNewlyDisabled :type Boolean) (bNewlyEnabled :type Boolean))
  ;[1 1 1] DamageablePartHealthChanged
  (CallFunction|DamageablePartChanged PartConfig Part bNewlyDisabled))

(event Vehicle|EventOnSpringCompression
  (bind _returnvalue (TouchPlatform|IsMobileGame))
  ;[1 1 1] EVENT -> OnSpringCompression - play force feedback / audio
  (if _returnvalue
    (Development|PrintString (Utilities|String|Append "Is Mobile? " (Utilities|String|ToString(Boolean) _returnvalue)))
    (else
      (Class|MotorcycleFunctionLibrary|SpringCompressionFF (Variables|PawnReferences|GetDrivingPlayerController) (Variables|ForceFeedback|GetSpringCompressionForceFeedback)))))

(event Custom|HandleGearChange
  (params (NewGear :type Integer))
  (bind _ns_dirtbike_roadandexhaust (Variables|Default|GetNS_DirtBike_RoadAndExhaust))
  ;[1 1 1] Check Current Gears
  (Class|MotorcycleFunctionLibrary|HandleGearChange NewGear _ns_dirtbike_roadandexhaust (Variables|PawnReferences|GetDriverPawn))
  (CallFunction|GearShiftFovChange NewGear)
  ;;[1 1 1] latent events cant happen in library functions
  (Utilities|FlowControl|RetriggerableDelay 0.2)
  (Effects|Components|ParticleSystem|SetBoolParameter _ns_dirtbike_roadandexhaust "IsShifting" false)
  ;;[1 1 1] Reset FOV
  (Variables|Camera|SetCameraFOVOffset 0.0))

(event Custom|EventOnHealthChanged
  (bind _returnvalue (Class|FortVehicleInterface|GetHealth))
  (bind _returnvalue_2 (Class|FortVehicleInterface|GetMaxHealth))
  (bind _returnvalue_3 (/ _returnvalue _returnvalue_2))
  (bind _returnvalue_4 (Math|Float|Lerp 1.0 0.0 _returnvalue_3))
  (bind _ns_dirtbike_roadandexhaust (Variables|Default|GetNS_DirtBike_RoadAndExhaust))
  ;[1 1 1] Update Material based on remaining health %
  (bind _output_get (Variables|DamageStates|SetDamageValue _returnvalue_4))
  (CallFunction|VFXDamageState _output_get)
  (CallFunction|UpdateDamageState _returnvalue_4)
  ;[1 1 1] Restore FX params at full health
  (if (>= (Class|FortVehicleInterface|GetHealth) (Class|FortVehicleInterface|GetMaxHealth))
    (bind _output_get_2 (Variables|States|SetTirePoppedFront false))
    (Niagara|SetNiagaraVariable(Bool) _ns_dirtbike_roadandexhaust "TirePoppedSparksFront" _output_get_2)
    (bind _output_get_3 (Variables|States|SetTirePoppedRear false))
    (Niagara|SetNiagaraVariable(Bool) _ns_dirtbike_roadandexhaust "TirePoppedSparksRear" _output_get_3)))

(event EventOnRimScrapeDamage
  (params (TireIndex :type Integer))
  ;[1 1 1] EVENT -> Rim Scrape Damage
  (Class|MotorcycleFunctionLibrary|HandleScrapeDamage (Variables|GameplayEffects|GetGE_RimScrapeDamage) (Vehicle|GetVehicleAbilitySystemComponent) (or (Class|FortVehicleInterface|IsinWater) (Vehicle|VehicleIsinAir))))

(event Custom|Update%20Damage%20State
  (params (Damage :type Float%20%28double-precision%29))
  ;[1 1 1] Update Material based on remaining health %
  (CallFunction|SendDamageValuetoMaterials Damage))

(event EventPawnHitByVehicle
  (params (Pawn :type Fort%20Pawn%20Object%20Reference) (ImpactPoint :type Vector%20%28by%20ref%29 :default "0, 0, 0") (ImpactNormal :type Vector%20%28by%20ref%29 :default "0, 0, 0") (LaunchVector :type Vector :default "0, 0, 0") (bTeleportPawn :type Boolean) (TeleportPawnDistance :type Float%20%28single-precision%29))
  ;[1 1 1] EVENT -> Hit Pawn
  (Class|MotorcycleFunctionLibrary|OnImpactEvent Pawn (Vehicle|HasDriver) (Variables|States|GetPassengerIsInstigator) (Variables|PawnReferences|GetPassengers) (Variables|PawnReferences|GetLastDriver) (Variables|PawnReferences|GetDriverPawn) LaunchVector (Vehicle|GetAllPassengers) (Variables|Impact|GetHitPlayerCue) (Variables|Impact|GetHitFiendCue)))

(event Custom|SurfaceTypeChange
  (params (SurfaceTypeVehicleIsOn :type EPhysicalSurface%20Enum :default "SurfaceType_Default"))
  (bind _returnvalue (Math|Conversions|ToInteger(Byte) SurfaceTypeVehicleIsOn))
  ;[1 1 1] SurfaceTypeChange
  (bind _output_get (Variables|SurfaceType|SetSurface%20Type%20Vehicle%20Is%20On SurfaceTypeVehicleIsOn))
  (Class|MotorcycleFunctionLibrary|SurfaceTypeChange _output_get (Variables|Default|GetNS_DirtBike_RoadAndExhaust) (Vehicle|GetVehicleAbilitySystemComponent) (Variables|SurfaceType|GetGE_SurfaceIce))
  ;;[1 0 0.3957] Ice sticks on??
  (CallFunction|HandleSurfaceTypeChanged _output_get)
  ;[1 0.4768 0.6145] VFX: Short Mud Track when surface changed from Mud to other surface
  (if (== _returnvalue 30)
    (Variables|Default|SetPhysSurfaceIndex 30)
    (elif (== (Variables|Default|GetPhysSurfaceIndex) 30)
      (Niagara|SpawnSystemAttached "/Game/Effects/MudInteraction/NS_Mud_Track_Bike.NS_Mud_Track_Bike" (Variables|Vehicle|GetSkeletalMesh) "root" "0, 0, 0" "0, 0, 0" "KeepRelativeOffset" true true "AutoRelease")
      (else
        (Variables|Default|SetPhysSurfaceIndex _returnvalue)))))

(event Custom|VFX_DamageState
  (params (Damage :type Float%20%28double-precision%29))
  (bind _damagefx (Variables|Default|GetDamageFX))
  (bind _damagefx_2 (Variables|Default|GetDamageFX))
  ;[0 0 0] Turn on damage state Niagara vfx
  (if (>= Damage (Variables|Audio|GetDamageStage1))
    (Utilities|IsValid _damagefx_2
      (:Is%20Valid
        _)
      (:Is%20Not%20Valid
        (bind _returnvalue (Niagara|SpawnSystemAttached (Variables|FX|GetDamageFX_Template) (Variables|Vehicle|GetSkeletalMesh) "Engine" "0, 0, 0" "0, 0, 0" "KeepRelativeOffset" true))
        (Variables|Default|SetDamageFX _returnvalue)))
    (Niagara|SetNiagaraVariable(Float) _damagefx_2 "User.DamageScale" Damage)
    (else
      (Utilities|IsValid _damagefx
        (:Is%20Valid
          (Niagara|SetNiagaraVariable(Float) _damagefx "User.DamageScale" Damage))) ; Still need to scale FX if it's active and we are healed beyond min threshold
    )))

(event EventTick
  (params (DeltaSeconds :type Float%20%28single-precision%29)))

(event EventOnPawnExitVehicle
  (params (PlayerPawn :type Fort%20Player%20Pawn%20Object%20Reference) (ExitSocketName :type Name :default "None"))
  (bind _returnvalue (Ability|GetFortAbilitySystemComponent PlayerPawn))
  ;[1 1 1] Pawn Exit Vehicle
  (Variables|States|SetisVehicleOccupied? (CallFunction|IsVehicleOccupied))
  (Utilities|Array|RemoveItem (Variables|PawnReferences|GetPassengers) PlayerPawn)
  (CallFunction|CEPawnExitVehicleWhileinWater)
  (bind _asplayer_controller (Utilities|Casting|CastToPlayerController (Pawn|GetController PlayerPawn))
    (:then
      (Utilities|Array|RemoveItem (Variables|PawnReferences|GetValidPlayerControllers) _asplayer_controller))
    (:CastFailed
      _))
  (if (Networking|IsServer)
    (GameplayEffects|RemoveActiveEffectswithGrantedTags _returnvalue (Variables|GameplayTags|GetInDirtbikeVehicleTag))
    (GameplayEffects|RemoveActiveGameplayEffectbySourceEffect _returnvalue (Variables|GameplayTags|GetGE_ApplyAllowFlippingTag) 0 -1)))

(event Custom|HandleOnExplode
  (params (LastDamageInstigator :type Controller%20Object%20Reference) (Vehicle :type Actor%20Object%20Reference))
  (bind _surface_type_vehicle_is_on (Variables|SurfaceType|GetSurface%20Type%20Vehicle%20Is%20On))
  (bind _returnvalue (Math|Conversions|ToInteger(Byte) _surface_type_vehicle_is_on))
  ;[1 0.1458 0.1678] EVENT -> On Explode
  (Variables|PawnReferences|SetLast%20Damage%20Instigator LastDamageInstigator)
  (bind _asfort_athena_vehicle (Utilities|Casting|CastToFortAthenaVehicle Vehicle))
  (Variables|PawnReferences|SetVehicle _asfort_athena_vehicle)
  (Utilities|FlowControl|Sequence
    (:then_0
      ;;[0 0 0] Trigger Explosion Cosmetics
      (if (Networking|IsDedicatedServer)
        (CallFunction|DestroyAllDamageableParts)
        ;;[1 1 1] Damage is happening twice on server??
        (Utilities|FlowControl|DoOnce false
          (:Completed
            (Class|VehicleFunctionLibrary|VehicleExplodeandDamageNearbyActors (Variables|PawnReferences|GetVehicle) (Variables|PawnReferences|GetLast%20Damage%20Instigator) (ScalableFloat|GetValueatLevel (Variables|Hotfix|GetExplodeRadius)) (Variables|GameplayEffects|GetGE_ExplodeVehicleDamage) (Variables|GameplayEffects|GetGE_Explode_BuildingDamage) (Variables|GameplayEffects|GetGE_ExplodePlayerDamage) (ScalableFloat|GetValueatLevel (Variables|Hotfix|GetExplosionDamageVehicles)) (ScalableFloat|GetValueatLevel (Variables|Hotfix|GetExplosionDamagePlayers)) (ScalableFloat|GetValueatLevel (Variables|Hotfix|GetExplosionDamageBuildings)) false true (ScalableFloat|GetValueatLevel (Variables|Hotfix|GetExplodeVehicleImpulse)) (ScalableFloat|GetValueatLevel (Variables|Hotfix|GetExplodePlayerImpulse)))))
        (else
          (Class|MotorcycleFunctionLibrary|ExplosionFX (Variables|FX|GetExplodeFXNiagara) (Transformation|GetActorLocation) (Variables|Vehicle|GetSkeletalMesh) (Variables|FX|GetWreckFXNiagara) (Variables|CameraShake|GetExplosion_CameraShake_Radius) (CallFunction|GetMaxZeroHealthTime) (Variables|CameraShake|GetExplosionShake) (Variables|DamageStates|GetDestructionMaterial) 0.6)
          (bind _checkedphysicalsurfaceid (CallFunction|PhysicalSurfaceIDCheck _returnvalue))
          (if _checkedphysicalsurfaceid
            ;;[1 0.1458 0.1678] VFX Explosion  Phisical Surface Reaction
            (bind _returnvalue_2 (Niagara|SpawnSystemAttached "/Game/Effects/MudInteraction/NS_Physical_Surface_Explosion.NS_Physical_Surface_Explosion" (Variables|Vehicle|GetSkeletalMesh)))
            (Effects|Components|ParticleSystem|SetIntParameter _returnvalue_2 "PhysicalSurface" _returnvalue)
            (Effects|Components|ParticleSystem|SetFloatParameter _returnvalue_2 "Global Scale" 1.0)
            ;;;[0.2923 0.4974 1] Setting for Niagara Particle Data Export
            (Niagara|SetNiagaraVariable(Object) _returnvalue_2 "Export particle data" self))
          (CallFunction|SpawnFNRadialForce (Transformation|GetWorldTransform (Variables|Vehicle|GetSkeletalMesh)) 0.3 0.074 2.4e+03 1.75))))
    (:then_1
      (bind _damagefx (Variables|Default|GetDamageFX)
        (:then
          (Components|Activation|Deactivate _damagefx))
        (:else)))))

(event Custom|LoadAudioControllerClass
  ;[1 1 1] Async load audio controller REV
  (Networking|CanExecuteCosmeticEvents
    (:True
      (bind _audiodataasset (Variables|Audio|GetAudioDataAsset)
        (:then
          (bind _loaded_class (Creative|Loading|AsyncLoadClassEditorSafe (Variables|/VehicleAudio/B_Vehicle_AudioController_PrimaryDataAsset.B_Vehicle_AudioController_PrimaryDataAsset_C|GetAudio%20Controller%20Class _audiodataasset))
            (:Loaded
              (bind _asactor (Utilities|Casting|CastToActorClass _loaded_class))
              (bind _motorcycle_audiocontroller (Variables|Default|GetMotorcycle_AudioController)
                (:then
                  (ChildActorComponent|SetChildActorClass _motorcycle_audiocontroller _asactor)
                  (bind _asbp_vehicles_audio_controller_motorcycle_parent (Utilities|Casting|CastToBP_Vehicles_AudioController_Motorcycle_Parent (Variables|/Script/Engine.ChildActorComponent|GetChildActor (Variables|Default|GetMotorcycle_AudioController))))
                  (Class|BPVehicleCosmeticsAudioControllerParent|SetAudioDataAsset _asbp_vehicles_audio_controller_motorcycle_parent (Variables|Audio|GetAudioDataAsset) :!class /VehicleCosmeticsAudio/Controllers/BP_VehicleCosmetics_AudioController_Parent.BP_VehicleCosmetics_AudioController_Parent_C))
                (:else)))))
        (:else)))))

(event EventOnPawnEnterVehicle
  (params (PlayerPawn :type Fort%20Player%20Pawn%20Object%20Reference) (SeatIdx :type Integer))
  (bind _returnvalue (Ability|GetAbilitySystemComponent PlayerPawn))
  (bind _ns_dirtbike_roadandexhaust (Variables|Default|GetNS_DirtBike_RoadAndExhaust))
  ;[1 1 1] Pawn Enter Vehicle
  (CallFunction|CEPawnEnterVehicleWhileinWater)
  (Variables|States|SetisVehicleOccupied? true)
  (Ability|SendGameplayEventtoActor PlayerPawn "(TagName=\"Event.Vehicle.Entered\")" (Utilities|Struct|MakeGameplayEventData :Instigator self :OptionalObject self :ContextHandle "()" :InstigatorTags "(GameplayTags=)" :TargetTags "(GameplayTags=)" :EventMagnitude 0.0 :TargetData "()"))
  (CallFunction|AddtoPlayerControllerArray PlayerPawn)
  (Utilities|Array|AddUnique (Variables|PawnReferences|GetPassengers) PlayerPawn)
  (Niagara|SetNiagaraVariable(Bool) _ns_dirtbike_roadandexhaust "TirePoppedSparksFront" (Variables|States|GetTirePoppedFront))
  (Niagara|SetNiagaraVariable(Bool) _ns_dirtbike_roadandexhaust "TirePoppedSparksRear" (Variables|States|GetTirePoppedRear))
  (if (Networking|IsServer)
    (GameplayEffects|ApplyGameplayEffectToTarget :self _returnvalue :GameplayEffectClass (Variables|GameplayTags|GetGE_ApplyInDirtbikeTag) :Target _returnvalue :Level 1.0)
    (if (GameplayTags|HasAnyMatchingGameplayTags (Variables|GameplayTags|GetVehicleAllowFlippingTag))
      (GameplayEffects|ApplyGameplayEffectToTarget :self _returnvalue :GameplayEffectClass (Variables|GameplayTags|GetGE_ApplyAllowFlippingTag) :Target _returnvalue :Level 1.0))))

(event Custom|DriverHUDVisibility
  (params (bHideElements :type Boolean))
  ;[1 1 1] Set visibility vehicle guage HUD
  (Class|MotorcycleFunctionLibrary|HideHUD (Fortnite|GetFortPlayerControllerfromActor (Variables|PawnReferences|GetDriverPawn)) (Variables|HUD|GetDriverHUDTag) bHideElements))

(event Niagara|EventReceiveParticleData
  (params (Data :type Array%20of%20Basic%20Particle%20Data%20Structures) (NiagaraSystem :type Niagara%20System%20Object%20Reference) (SimulationPositionOffset :type Vector%20%28by%20ref%29 :default "0, 0, 0"))
  ;[1 0.1458 0.1678] EVENT -> On Explode
  ;[1 0.1458 0.1678] Event VFX: Spawn FX from collided particles. for Mud ( for now)
  (for _array_element Data
    (bind (_returnvalue_roll _returnvalue_pitch _returnvalue_yaw) (Math|Rotator|BreakRotator (Math|Conversions|RotationFromXVector (.Velocity (Utilities|Struct|BreakBasicParticleData _array_element)))))
    (Niagara|SpawnSystematLocation "/Game/Effects/MudInteraction/NS_Mud_Splash_Burst_Proto.NS_Mud_Splash_Burst_Proto" (Utilities|Struct|BreakBasicParticleData _array_element) (Math|Rotator|MakeRotator _returnvalue_roll (+ _returnvalue_pitch -9e+01) _returnvalue_yaw))))

(event Custom|InventoryHUDVisibility
  (params (bHideElements :type Boolean))
  ;[1 1 1] Set visibility of inventory HUD
  (Class|MotorcycleFunctionLibrary|HideHUD (Fortnite|GetFortPlayerControllerfromActor (Variables|PawnReferences|GetDriverPawn)) (Variables|HUD|GetQuickbarHUDTag) bHideElements))

(event EventBeginPlay
  ;[1 1 1] EVENT -> Begin Play
  (Utilities|FlowControl|Sequence
    (:then_0
      (Vehicle|BindEventtoOnPawnEnteredVehicleSeat (EventDispatchers|CreateEvent :!event Custom|PawnEnterSeat))
      (Vehicle|BindEventtoOnPawnExitedVehicleSeat (EventDispatchers|CreateEvent :!event Custom|PawnExitSeat))
      (CallFunction|LoadAudioControllerClass)
      (Game|BindEventtoOnDestroyed (delegate Custom|FadeoutCosmetics)) ; Need this to get a client side event when destroyed.
      (CallFunction|PickRandomMaterial)
      (Collision|BindEventtoOnDamageablePartHealthChanged (EventDispatchers|CreateEvent :!event Custom|Event_OnDamageablePartHealthChanged))
      (SurfaceType|BindEventtoOnSurfaceTypeVehicleIsonChangedDelegate (EventDispatchers|CreateEvent :!event Custom|SurfaceTypeChange))
      (Variables|States|SetIgnoreHitLastDriver true)
      (Variables|States|SetPassengerIsInstigator true)
      (EventDispatchers|BindEventtoOnRefueledfromEmptyDelegate (EventDispatchers|CreateEvent :!event HandleRefueledFromEmpty))
      (CallFunction|SetPropDestructionParams (Utilities|Array|MakeArray "/Script/FortniteGame.BuildingProp" "/Game/Building/ActorBlueprints/Containers/ParentBuildingContainerBlueprint.ParentBuildingContainerBlueprint_C") (Utilities|Array|MakeArray) (Utilities|Array|MakeArray "/Game/Building/EditModePatterns/Wall/EMP_Wall_HalfWallS.EMP_Wall_HalfWallS") (Utilities|Array|MakeArray))
      (CallFunction|CacheExplosionDamageValues)
      ;;[1 1 1] EVENT -> CurieContainerAcquired
      (Health|BindEventtoOnHealthChanged (EventDispatchers|CreateEvent :!event Custom|EventOnHealthChanged))
      (Vehicle|Destruction|BindEventtoOnVehicleExplodeDelegate (EventDispatchers|CreateEvent :!event Custom|HandleOnExplode))
      (Vehicle|Destruction|BindEventtoOnStartCriticalHealth (EventDispatchers|CreateEvent :!event Custom|HandleOnStartCriticalHealth)))
    (:then_1
      (Vehicle|AssignOnCurrentGearChangedDelegate (EventDispatchers|CreateEvent :!event Custom|HandleGearChange))
      (Class|FortAsyncActionGameStateReady|GameStateReadyAsync
        (:OnReady
          (if (and (Spawn|GetForcedtoAlwaysSpawn) (Networking|IsServer))
            (Vehicle|Fuel|SetFuel 2e+02)))))))

(event Custom|FadeoutCosmetics
  (params (DestroyedActor :type Actor%20Object%20Reference))
  ;[1 1 1] EVENT -> Begin Play
  ;;[1 1 1] Fadeout FX
  (Rendering|SetVisibility (Variables|Vehicle|GetSkeletalMesh) false false))

(event Curie|EventOnCurieContainerAcquired
  (params (CurieContainerHandle :type Curie%20Container%20Handle%20Structure))
  ;[1 1 1] EVENT -> Begin Play
  ;;[1 1 1] EVENT -> CurieContainerAcquired
  (if (not (CallFunction|IsonFire))
    (Curie|DisableElementAttachmentonActor self "(TagName=\"Curie.Element.Fire\")")))

(event EventOnSleepStateChanged
  (params (bAwake :type Boolean))
  ;[1 1 1 1 25] EVENT -> SleepStateChanged
  (if (Networking|IsDedicatedServer)
    _
    (elif bAwake
      (CallFunction|FXInitSettings)
      (else
        (Components|Activation|Deactivate (Variables|Default|GetNS_DirtBike_RoadAndExhaust))))))

(event Custom|PawnEnterSeat
  (params (Vehicle :type Fort%20Vehicle%20Interface%20Interface%20%28by%20ref%29) (PlayerPawn :type Fort%20Pawn%20Object%20Reference) (SeatIndex :type Integer))
  ;[1 1 1] Pawn Enter Seat
  (bind _asfort_player_pawn (Utilities|Casting|CastToFortPlayerPawn PlayerPawn))
  (switch int SeatIndex
    (:0
      ;; ManageStateDriverEnter
      (bind _output_get (Variables|PawnReferences|SetDriverPawn _asfort_player_pawn))
      (Utilities|FlowControl|Sequence
        (:then_0
          (Variables|PawnReferences|GetPassengerPawn
            (:then
              (CallFunction|HideBackBling (Variables|PawnReferences|GetDriverPawn)))
            (:else)))
        (:then_1
          (bind _asplayer_controller (Utilities|Casting|CastToPlayerController (Pawn|GetController _output_get)))
          (Variables|PawnReferences|SetDrivingPlayerController _asplayer_controller)
          (CallFunction|CanHitLastDriver)
          (Variables|States|SetPassengerIsInstigator false))
        (:then_2
          (Variables|Camera|SetCameraSpaceForwardDistanceOffset 0.0)
          (bind _returnvalue (CallFunction|IsDriverRestrictedbyTeamSettings _output_get))
          (if _returnvalue
            _)))
      (Niagara|SetNiagaraVariable(Bool) (Variables|Default|GetNS_DirtBike_RoadAndExhaust) "VehicleHasDriver" true))
    (:1
      _)
    (:2
      _)
    (:3
      _))
  ;; ManageStatePassengerEnter
  (Variables|PawnReferences|SetPassengerPawn _asfort_player_pawn)
  (if (Vehicle|HasDriver)
    (CallFunction|HideBackBling (Variables|PawnReferences|GetDriverPawn))
    (else
      _)))

(event Custom|HideBackBling
  (params (Pawn :type Fort%20Player%20Pawn%20Object%20Reference) (ShouldntHide :type Boolean))
  ;[1 1 1] Hide Back Bling
  (Variables|States|SetBackPackIsntHidden ShouldntHide)
  (if (not (Networking|IsServer))
    (Utilities|IsValid Pawn
      (:Is%20Valid
        (CharacterPart|SetCharacterPartVisibility Pawn "Backpack" ShouldntHide true "()")))))

(event Custom|FX_Init_Settings
  ;[1 1 1 1 25] EVENT -> SleepStateChanged
  ;;[1 1 1] FX_Init_Settings
  (CallFunction|SetFxSettingsforVehicleType))

(event Custom|PawnExitSeat
  (params (Vehicle :type Fort%20Vehicle%20Interface%20Interface%20%28by%20ref%29) (PlayerPawn :type Fort%20Pawn%20Object%20Reference) (SeatIndex :type Integer))
  ;[1 1 1] Pawn Exit Seat
  (bind _asfort_player_pawn (Utilities|Casting|CastToFortPlayerPawn PlayerPawn))
  (switch int SeatIndex
    (:0
      (if (not (Variables|States|GetBackPackIsntHidden))
        (CallFunction|HideBackBling (Variables|PawnReferences|GetDriverPawn) true))
      (Utilities|FlowControl|Sequence
        (:then_0
          ;; ManageStateDriverExit
          (Utilities|FlowControl|Sequence
            (:then_0
              (Variables|States|SetIgnoreHitLastDriver true)
              (Utilities|Time|SetTimerbyEvent (EventDispatchers|CreateEvent :!event CanHitLastDriver) 0.6)
              (bind _returnvalue (CallFunction|IsDriverRestrictedbyTeamSettings _asfort_player_pawn))
              (if _returnvalue
                _)
              (Utilities|FlowControl|Sequence
                (:then_0
                  (Class|MotorcycleFunctionLibrary|ApplyBoostGC false (Variables|PawnReferences|GetDriverPawn) (Variables|Boost|GetBoost%20GC)))
                (:then_1
                  (Variables|PawnReferences|SetDrivingPlayerController 0))))
            (:then_1
              (Variables|PawnReferences|SetDriverPawn 0)
              (Variables|PawnReferences|SetLastDriver _asfort_player_pawn))
            (:then_2
              _)))
        (:then_1
          _)))
    (:1
      (Variables|PawnReferences|SetPassengerPawn 0)
      (CallFunction|HideBackBling (Variables|PawnReferences|GetDriverPawn) true)))
  (Niagara|SetNiagaraVariable(Bool) (Variables|Default|GetNS_DirtBike_RoadAndExhaust) "VehicleHasDriver" (Vehicle|HasDriver)))

(event EventOnWaterTooDeep
  ;[0.6084 0.8133 1] EVENT -> Enter/ExitWater:  Niagara FX and Audio for Water Interactions
  ;;[0 0 0] Kick players + deactive dmg vfx when too deep in water
  (|CallParentFunction))

(event EventOnEnteredWaterBody
  (params (WaterBody :type Fort%20Water%20Body%20Actor%20Object%20Reference))
  (bind _watersurfacedatacache (Variables|FortAthenaVehicle|GetWaterSurfaceDataCache))
  (bind (_waterplanelocation _waterplanenormal _watersurfacelocation _waterdepth _waterbodyidx) (Utilities|Struct|BreakWaterSurfaceInfo _watersurfacedatacache))
  (bind (_returnvalue_x _returnvalue_y _returnvalue_z) (Math|Vector|BreakVector (Transformation|GetActorLocation)))
  (bind (_returnvalue_x_2 _returnvalue_y_2 _returnvalue_z_2) (Math|Vector|BreakVector (Transformation|GetActorLocation)))
  (bind _ns_largesplashfx (Variables|Default|GetNS_LargeSplashFX))
  ;[0.6084 0.8133 1] EVENT -> Enter/ExitWater:  Niagara FX and Audio for Water Interactions
  (Variables|WaterInteraction|SetWaterBody WaterBody)
  (Effects|Components|ParticleSystem|SetBoolParameter (Variables|Default|GetNS_DirtBike_RoadAndExhaust) "IsInWater" true)
  (Utilities|FlowControl|DoOnce false
    (:Completed
      (bind _returnvalue (Utilities|Time|SetTimerbyEvent (delegate Custom|SnapFoamtoWaterSurface) 0.2 true))
      (Variables|WaterInteraction|SetTimerSnapFoamToWaterSurface _returnvalue)
      (if (not (Networking|IsServer))
        (bind _returnvalue_2 (Niagara|SpawnSystemAttached (Variables|FX|GetTemplateWaterSplashFX) (Variables|Vehicle|GetSkeletalMesh) "WaterTest" "0, 0, 0" "0, 0, 0" "KeepRelativeOffset" true))
        (Variables|Default|SetNS_LargeSplashFX _returnvalue_2)
        (Utilities|FlowControl|Sequence
          (:then_0
            ;;[0 0 0] Trigger large splash base on falling speed
            (Niagara|SetNiagaraVariable(Float) _ns_largesplashfx "LoopFoamSpawn" 0.0)
            (Variables|Impact|SetVehicleFallSpeed _returnvalue_z)
            (Utilities|FlowControl|Delay 0.08)
            (if (> (Math|Float|Absolute(Float) (- (Variables|Impact|GetVehicleFallSpeed) _returnvalue_z_2)) 1e+01)
              (Niagara|SetNiagaraVariable(Int32) _ns_largesplashfx "SplashSpawn" 1)
              (Niagara|SetNiagaraVariable(Float) _ns_largesplashfx "BubbleSpawn" 0.0)
              (Niagara|SetNiagaraVariable(Int32) _ns_largesplashfx "BigBubbleSpawn" 0)
              (Niagara|SetNiagaraVariable(Vector3) _ns_largesplashfx "WaterPlaneLocation" _watersurfacelocation)
              (Niagara|SetNiagaraVariable(Float) _ns_largesplashfx "WaterDepth" _waterdepth)
              (Niagara|SetNiagaraVariable(Vector3) _ns_largesplashfx "WaterPlaneNormal" _waterplanenormal)
              (Niagara|SetNiagaraVariable(Float) _ns_largesplashfx "WaterBodyIndex" (Math|Conversions|ToFloat(Integer) _waterbodyidx))
              (if (> (Rendering|GetRenderingDetailMode) 0)
                (bind _driverpawn (Variables|PawnReferences|GetDriverPawn)
                  (:then
                    (Feedback|SpawnGenericCameraLensEffect _driverpawn "/Game/Effects/Fort_Effects/Camera_Effects/B_CameraLens_Splash_Water_Droplets.B_CameraLens_Splash_Water_Droplets_C"))
                  (:else)))
              (else
                (Niagara|SetNiagaraVariable(Int32) _ns_largesplashfx "SplashSpawn" 0)
                (Niagara|SetNiagaraVariable(Float) _ns_largesplashfx "BubbleSpawn" 0.0)
                (Niagara|SetNiagaraVariable(Int32) _ns_largesplashfx "BigBubbleSpawn" 0))))
          (:then_1
            ;;[0 0 0] Deactivate Water bubble FX while sinking to certain point and no driver in vehicle
            (Utilities|FlowControl|Delay 12.0)
            (Variables|PawnReferences|GetDriverPawn
              (:then)
              (:else
                (Components|Activation|Deactivate _ns_largesplashfx)
                (Utilities|Time|ClearandInvalidateTimerbyHandle (Variables|WaterInteraction|GetTimerSnapFoamToWaterSurface))
                (Development|PrintString "DeactiveWaterFXwhenDriverIsNotValid")))))))))

(event Custom|CE_PawnEnterVehicleWhileInWater
  (bind _watersurfacedatacache (Variables|FortAthenaVehicle|GetWaterSurfaceDataCache))
  (bind (_waterplanelocation _waterplanenormal _watersurfacelocation _waterdepth _waterbodyidx) (Utilities|Struct|BreakWaterSurfaceInfo _watersurfacedatacache))
  (bind (_returnvalue_x _returnvalue_y _returnvalue_z) (Math|Vector|BreakVector (Transformation|GetActorLocation)))
  (bind (_returnvalue_x_2 _returnvalue_y_2 _returnvalue_z_2) (Math|Vector|BreakVector (Transformation|GetActorLocation)))
  (bind _ns_largesplashfx (Variables|Default|GetNS_LargeSplashFX))
  ;[0.6084 0.8133 1] EVENT -> Enter/ExitWater:  Niagara FX and Audio for Water Interactions
  (if (CallFunction|IsinWaterBody)
    (bind _ns_largesplashfx_2 (Variables|Default|GetNS_LargeSplashFX)
      (:then
        (Components|Activation|Activate _ns_largesplashfx_2 false))
      (:else
        (bind _returnvalue (Utilities|Time|SetTimerbyEvent (delegate Custom|SnapFoamtoWaterSurface) 0.2 true))
        (Variables|WaterInteraction|SetTimerSnapFoamToWaterSurface _returnvalue)
        (if (not (Networking|IsServer))
          (bind _returnvalue_2 (Niagara|SpawnSystemAttached (Variables|FX|GetTemplateWaterSplashFX) (Variables|Vehicle|GetSkeletalMesh) "WaterTest" "0, 0, 0" "0, 0, 0" "KeepRelativeOffset" true))
          (Variables|Default|SetNS_LargeSplashFX _returnvalue_2)
          (Utilities|FlowControl|Sequence
            (:then_0
              ;;[0 0 0] Trigger large splash base on falling speed
              (Niagara|SetNiagaraVariable(Float) _ns_largesplashfx "LoopFoamSpawn" 0.0)
              (Variables|Impact|SetVehicleFallSpeed _returnvalue_z)
              (Utilities|FlowControl|Delay 0.08)
              (if (> (Math|Float|Absolute(Float) (- (Variables|Impact|GetVehicleFallSpeed) _returnvalue_z_2)) 1e+01)
                (Niagara|SetNiagaraVariable(Int32) _ns_largesplashfx "SplashSpawn" 1)
                (Niagara|SetNiagaraVariable(Float) _ns_largesplashfx "BubbleSpawn" 0.0)
                (Niagara|SetNiagaraVariable(Int32) _ns_largesplashfx "BigBubbleSpawn" 0)
                (Niagara|SetNiagaraVariable(Vector3) _ns_largesplashfx "WaterPlaneLocation" _watersurfacelocation)
                (Niagara|SetNiagaraVariable(Float) _ns_largesplashfx "WaterDepth" _waterdepth)
                (Niagara|SetNiagaraVariable(Vector3) _ns_largesplashfx "WaterPlaneNormal" _waterplanenormal)
                (Niagara|SetNiagaraVariable(Float) _ns_largesplashfx "WaterBodyIndex" (Math|Conversions|ToFloat(Integer) _waterbodyidx))
                (if (> (Rendering|GetRenderingDetailMode) 0)
                  (bind _driverpawn (Variables|PawnReferences|GetDriverPawn)
                    (:then
                      (Feedback|SpawnGenericCameraLensEffect _driverpawn "/Game/Effects/Fort_Effects/Camera_Effects/B_CameraLens_Splash_Water_Droplets.B_CameraLens_Splash_Water_Droplets_C"))
                    (:else)))
                (else
                  (Niagara|SetNiagaraVariable(Int32) _ns_largesplashfx "SplashSpawn" 0)
                  (Niagara|SetNiagaraVariable(Float) _ns_largesplashfx "BubbleSpawn" 0.0)
                  (Niagara|SetNiagaraVariable(Int32) _ns_largesplashfx "BigBubbleSpawn" 0))))
            (:then_1
              ;;[0 0 0] Deactivate Water bubble FX while sinking to certain point and no driver in vehicle
              (Utilities|FlowControl|Delay 12.0)
              (Variables|PawnReferences|GetDriverPawn
                (:then)
                (:else
                  (Components|Activation|Deactivate _ns_largesplashfx)
                  (Utilities|Time|ClearandInvalidateTimerbyHandle (Variables|WaterInteraction|GetTimerSnapFoamToWaterSurface))
                  (Development|PrintString "DeactiveWaterFXwhenDriverIsNotValid"))))))))))

(event Custom|SnapFoamtoWaterSurface
  (bind _watersurfacedatacache (Variables|FortAthenaVehicle|GetWaterSurfaceDataCache))
  (bind (_waterplanelocation _waterplanenormal _watersurfacelocation _waterdepth _waterbodyidx) (Utilities|Struct|BreakWaterSurfaceInfo _watersurfacedatacache))
  (bind _returnvalue (Transformation|GetActorLocation))
  (bind _z (.z _returnvalue))
  (bind _z_2 (.z _waterplanelocation))
  (bind _returnvalue_2 (- _z_2 _z))
  (bind _ns_largesplashfx (Variables|Default|GetNS_LargeSplashFX))
  (bind _ns_largesplashfx_2 (Variables|Default|GetNS_LargeSplashFX))
  (bind _ns_largesplashfx_3 (Variables|Default|GetNS_LargeSplashFX))
  (bind _ns_dirtbike_roadandexhaust (Variables|Default|GetNS_DirtBike_RoadAndExhaust))
  ;[0.6084 0.8133 1] EVENT -> Enter/ExitWater:  Niagara FX and Audio for Water Interactions
  (if (Math|Float|InRange(Float) _returnvalue_2 -8e+01 8e+02)
    (bind _ns_largesplashfx_4 (Variables|Default|GetNS_LargeSplashFX)
      (:then
        (if (Math|Float|InRange(Float) _returnvalue_2 -39.900002 2.8e+02)
          (Variables|WaterInteraction|SetIsInDeepWater false)
          (Variables|PawnReferences|GetDriverPawn
            (:then
              ;;[0 0 0] Shallow Water
              (Components|Activation|Activate _ns_largesplashfx_4 false))
            (:else
              _))
          (Niagara|SetNiagaraVariable(Int32) _ns_largesplashfx_4 "SplashSpawn" 0)
          (Niagara|SetNiagaraVariable(Float) _ns_largesplashfx_4 "BubbleSpawn" 0.0)
          (Niagara|SetNiagaraVariable(Float) _ns_largesplashfx_4 "LoopFoamSpawn" 1.0)
          (Niagara|SetNiagaraVariable(Int32) _ns_largesplashfx_4 "BigBubbleSpawn" 0)
          (Niagara|SetNiagaraVariable(Bool) _ns_dirtbike_roadandexhaust "IsWaterDeep" false)
          ;;[0 0 0] Foam Snap Function
          (Utilities|FlowControl|Sequence
            (:then_0
              (Niagara|SetNiagaraVariable(Vector3) (Variables|Default|GetNS_LargeSplashFX) "WaterPlaneLocation" (Math|Vector|MakeVector (.x (Transformation|GetActorLocation)) (.y (Transformation|GetActorLocation)) (CallFunction|GetWaterLevel))))
            (:then_1
              (Niagara|SetNiagaraVariable(Float) _ns_largesplashfx_2 "WaterDepth" _waterdepth)
              (Niagara|SetNiagaraVariable(Vector3) _ns_largesplashfx_2 "WaterPlaneNormal" _waterplanenormal)
              (Niagara|SetNiagaraVariable(Float) _ns_largesplashfx_2 "WaterBodyIndex" (Math|Conversions|ToFloat(Integer) _waterbodyidx)))
            (:then_2
              ;;[0 0 0] Tire Spinning in water
              (if (Math|Float|InRange(Float) _returnvalue_2 -29.0 1e+02)
                (bind _ns_largesplashfx_5 (Variables|Default|GetNS_LargeSplashFX)
                  (:then
                    (Niagara|SetNiagaraVariable(Float) _ns_largesplashfx_5 "TireSpinWaterSpawn" 1.0)
                    (if (Vehicle|PlayerIsTryingtoMoveForward)
                      (Niagara|SetNiagaraVariable(Float) _ns_largesplashfx_3 "TireSpinWaterDirection" -1.0)
                      (elif (Vehicle|PlayerIsTryingtoReverse)
                        (Niagara|SetNiagaraVariable(Float) _ns_largesplashfx_3 "TireSpinWaterDirection" 1.0))))
                  (:else))
                (else
                  (bind _ns_largesplashfx_6 (Variables|Default|GetNS_LargeSplashFX)
                    (:then
                      (Niagara|SetNiagaraVariable(Float) _ns_largesplashfx_6 "TireSpinWaterSpawn" 0.0))
                    (:else))))
              (Niagara|SetNiagaraVariable(Float) _ns_largesplashfx_6 "TireSpinWaterSpawn" 0.0)))
          (else
            ;;[0 0 0] Deep Water
            (Variables|WaterInteraction|SetIsInDeepWater true)
            (Utilities|FlowControl|Sequence
              (:then_0
                (Utilities|FlowControl|DoOnce false
                  (:Completed
                    (Variables|PawnReferences|GetDriverPawn
                      (:then
                        (Components|Activation|Activate _ns_largesplashfx true)
                        (Development|PrintString "Submerged!!" true true "(R=1.000000,G=0.000000,B=0.027806,A=1.000000)" 5.0)
                        (Niagara|SetNiagaraVariable(Bool) _ns_dirtbike_roadandexhaust "IsWaterDeep" true))
                      (:else)))))
              (:then_1
                (Niagara|SetNiagaraVariable(Int32) _ns_largesplashfx "SplashSpawn" 0)
                (Niagara|SetNiagaraVariable(Float) _ns_largesplashfx "BubbleSpawn" 1.0)
                (Niagara|SetNiagaraVariable(Float) _ns_largesplashfx "LoopFoamSpawn" 0.0)
                (Niagara|SetNiagaraVariable(Int32) _ns_largesplashfx "BigBubbleSpawn" 1))))))
      (:else))
    (else
      (Components|Activation|Deactivate _ns_largesplashfx))))

(event Custom|CE_WaterFX_ResetDeepWaterEffectActivate
  ;[0.6084 0.8133 1] EVENT -> Enter/ExitWater:  Niagara FX and Audio for Water Interactions
  ;;[0 0 0] Deep Water
  (Utilities|FlowControl|DoOnce false
    (:Completed
      (Variables|PawnReferences|GetDriverPawn
        (:then
          (Components|Activation|Activate (Variables|Default|GetNS_LargeSplashFX) true)
          (Development|PrintString "Submerged!!" true true "(R=1.000000,G=0.000000,B=0.027806,A=1.000000)" 5.0)
          (Niagara|SetNiagaraVariable(Bool) (Variables|Default|GetNS_DirtBike_RoadAndExhaust) "IsWaterDeep" true))
        (:else)))))

(event EventOnExitedWaterBody
  (params (WaterBody :type Fort%20Water%20Body%20Actor%20Object%20Reference))
  (bind _watersurfacedatacache (Variables|FortAthenaVehicle|GetWaterSurfaceDataCache))
  (bind (_waterplanelocation _waterplanenormal _watersurfacelocation _waterdepth _waterbodyidx) (Utilities|Struct|BreakWaterSurfaceInfo _watersurfacedatacache))
  (bind (_returnvalue_x _returnvalue_y _returnvalue_z) (Math|Vector|BreakVector (Transformation|GetActorLocation)))
  (bind (_returnvalue_x_2 _returnvalue_y_2 _returnvalue_z_2) (Math|Vector|BreakVector (Transformation|GetActorLocation)))
  (bind _ns_largesplashfx (Variables|Default|GetNS_LargeSplashFX))
  ;[0.6084 0.8133 1] EVENT -> Enter/ExitWater:  Niagara FX and Audio for Water Interactions
  ;;[0 0 0] EVENT -> On Exited Water Body: Deactivate water FX When vehicle is out of water
  (Effects|Components|ParticleSystem|SetBoolParameter (Variables|Default|GetNS_DirtBike_RoadAndExhaust) "IsInWater" false)
  (Utilities|FlowControl|Sequence
    (:then_0
      (Utilities|Time|ClearandInvalidateTimerbyHandle (Variables|WaterInteraction|GetTimerSnapFoamToWaterSurface)))
    (:then_1
      (bind _ns_largesplashfx_2 (Variables|Default|GetNS_LargeSplashFX)
        (:then
          (Components|Activation|Deactivate _ns_largesplashfx_2))
        (:else)))
    (:then_2
      (Variables|PawnReferences|GetDriverPawn
        (:then
          (Utilities|FlowControl|DoOnce false
            (:Completed
              (bind _returnvalue (Utilities|Time|SetTimerbyEvent (delegate Custom|SnapFoamtoWaterSurface) 0.2 true))
              (Variables|WaterInteraction|SetTimerSnapFoamToWaterSurface _returnvalue)
              (if (not (Networking|IsServer))
                (bind _returnvalue_2 (Niagara|SpawnSystemAttached (Variables|FX|GetTemplateWaterSplashFX) (Variables|Vehicle|GetSkeletalMesh) "WaterTest" "0, 0, 0" "0, 0, 0" "KeepRelativeOffset" true))
                (Variables|Default|SetNS_LargeSplashFX _returnvalue_2)
                (Utilities|FlowControl|Sequence
                  (:then_0
                    ;;[0 0 0] Trigger large splash base on falling speed
                    (Niagara|SetNiagaraVariable(Float) _ns_largesplashfx "LoopFoamSpawn" 0.0)
                    (Variables|Impact|SetVehicleFallSpeed _returnvalue_z)
                    (Utilities|FlowControl|Delay 0.08)
                    (if (> (Math|Float|Absolute(Float) (- (Variables|Impact|GetVehicleFallSpeed) _returnvalue_z_2)) 1e+01)
                      (Niagara|SetNiagaraVariable(Int32) _ns_largesplashfx "SplashSpawn" 1)
                      (Niagara|SetNiagaraVariable(Float) _ns_largesplashfx "BubbleSpawn" 0.0)
                      (Niagara|SetNiagaraVariable(Int32) _ns_largesplashfx "BigBubbleSpawn" 0)
                      (Niagara|SetNiagaraVariable(Vector3) _ns_largesplashfx "WaterPlaneLocation" _watersurfacelocation)
                      (Niagara|SetNiagaraVariable(Float) _ns_largesplashfx "WaterDepth" _waterdepth)
                      (Niagara|SetNiagaraVariable(Vector3) _ns_largesplashfx "WaterPlaneNormal" _waterplanenormal)
                      (Niagara|SetNiagaraVariable(Float) _ns_largesplashfx "WaterBodyIndex" (Math|Conversions|ToFloat(Integer) _waterbodyidx))
                      (if (> (Rendering|GetRenderingDetailMode) 0)
                        (bind _driverpawn (Variables|PawnReferences|GetDriverPawn)
                          (:then
                            (Feedback|SpawnGenericCameraLensEffect _driverpawn "/Game/Effects/Fort_Effects/Camera_Effects/B_CameraLens_Splash_Water_Droplets.B_CameraLens_Splash_Water_Droplets_C"))
                          (:else)))
                      (else
                        (Niagara|SetNiagaraVariable(Int32) _ns_largesplashfx "SplashSpawn" 0)
                        (Niagara|SetNiagaraVariable(Float) _ns_largesplashfx "BubbleSpawn" 0.0)
                        (Niagara|SetNiagaraVariable(Int32) _ns_largesplashfx "BigBubbleSpawn" 0))))
                  (:then_1
                    ;;[0 0 0] Deactivate Water bubble FX while sinking to certain point and no driver in vehicle
                    (Utilities|FlowControl|Delay 12.0)
                    (Variables|PawnReferences|GetDriverPawn
                      (:then)
                      (:else
                        (Components|Activation|Deactivate _ns_largesplashfx)
                        (Utilities|Time|ClearandInvalidateTimerbyHandle (Variables|WaterInteraction|GetTimerSnapFoamToWaterSurface))
                        (Development|PrintString "DeactiveWaterFXwhenDriverIsNotValid")))))))))
        (:else)))
    (:then_3
      (if (not (CallFunction|IsOverlappingWaterBody))
        (Variables|WaterInteraction|SetWaterBody 0)))))

(event Custom|CE_PawnExitVehicleWhileInWater
  ;[0.6084 0.8133 1] EVENT -> Enter/ExitWater:  Niagara FX and Audio for Water Interactions
  ;;[0 0 0] EVENT -> On Exited Water Body: Deactivate water FX When vehicle is out of water
  (Utilities|IsValid (stale EventOnExitedWaterBody _ns_largesplashfx_2)))

(event EventOnWheelieStarted
  (bind _ns_dirtbike_roadandexhaust (Variables|Default|GetNS_DirtBike_RoadAndExhaust))
  ;[1 1 1] Tricks Events
  (Effects|Components|ParticleSystem|SetBoolParameter _ns_dirtbike_roadandexhaust "IsOnWheelie" true)
  (if (Variables|States|GetTirePoppedFront)
    (Niagara|SetNiagaraVariable(Bool) _ns_dirtbike_roadandexhaust "TirepoppedSparksFront" false)))

(event EventOnWheelieLanded
  (bind _ns_dirtbike_roadandexhaust (Variables|Default|GetNS_DirtBike_RoadAndExhaust))
  ;[1 1 1] Tricks Events
  (Utilities|FlowControl|RetriggerableDelay 0.5)
  (Effects|Components|ParticleSystem|SetBoolParameter _ns_dirtbike_roadandexhaust "IsOnWheelie" false)
  (if (Variables|States|GetTirePoppedFront)
    (Niagara|SetNiagaraVariable(Bool) _ns_dirtbike_roadandexhaust "TirepoppedSparksFront" true)))

(event EventOnStoppieStarted
  (bind _ns_dirtbike_roadandexhaust (Variables|Default|GetNS_DirtBike_RoadAndExhaust))
  ;[1 1 1] Tricks Events
  (Variables|States|SetIsDoingStoppie true)
  (Effects|Components|ParticleSystem|SetBoolParameter _ns_dirtbike_roadandexhaust "IsDoingStoppie" (Variables|States|GetIsDoingStoppie))
  (if (Variables|States|GetTirePoppedRear)
    (Niagara|SetNiagaraVariable(Bool) _ns_dirtbike_roadandexhaust "TirepoppedSparksRear" false)
    (else
      _))
  (Variables|Camera|SetCameraFOVOffset (Variables|Jump/Tricks|GetStoppieFOVOffset))
  ;[1 1 1] If vehicle reaches 0 speed reset camera
  (bind _returnvalue (Utilities|Time|SetTimerbyEvent (delegate Custom|ResetStoppieCamera) 0.1 true))
  (Variables|Jump/Tricks|SetNotMovingTimerHandle _returnvalue))

(event Custom|ResetStoppieCamera
  ;[1 1 1] If vehicle reaches 0 speed reset camera
  (if (<= (Vehicle|GetForwardSpeedKmh) 1e+01)
    (Variables|Camera|SetCameraFOVOffset (Variables|Jump/Tricks|GetDefault%20FOVOffset))))

(event EventOnStoppieLanded
  (bind _ns_dirtbike_roadandexhaust (Variables|Default|GetNS_DirtBike_RoadAndExhaust))
  ;[1 1 1] Tricks Events
  (Variables|States|SetIsDoingStoppie false)
  (Variables|Camera|SetCameraFOVOffset (Variables|Jump/Tricks|GetDefault%20FOVOffset))
  (Utilities|Time|ClearandInvalidateTimerbyHandle (Variables|Jump/Tricks|GetNotMovingTimerHandle))
  (Utilities|FlowControl|RetriggerableDelay 0.5)
  (Effects|Components|ParticleSystem|SetBoolParameter _ns_dirtbike_roadandexhaust "IsDoingStoppie" (Variables|States|GetIsDoingStoppie))
  (if (Variables|States|GetTirePoppedRear)
    (Niagara|SetNiagaraVariable(Bool) _ns_dirtbike_roadandexhaust "TirepoppedSparksRear" true)))

(event EventOnCollisionHitEffects
  (params (HitLocation :type Vector%20%28by%20ref%29 :default "0, 0, 0") (HitNormalImpulse :type Vector%20%28by%20ref%29 :default "0, 0, 0") (HitFrictionImpulse :type Vector%20%28by%20ref%29 :default "0, 0, 0") (HitNormal :type Vector%20%28by%20ref%29 :default "0, 0, 0") (HitActor :type Actor%20Object%20Reference) (HitSurfaceType :type EPhysicalSurface%20Enum :default "SurfaceType_Default"))
  (bind _returnvalue (Math|Vector|VectorLength HitNormalImpulse))
  (bind _returnvalue_2 (Math|Conversions|RotationFromXVector HitNormal))
  (bind _returnvalue_3 (Math|Rotator|InvertRotator _returnvalue_2))
  (bind (_asfort_pawn _bsuccess) (Utilities|Casting|CastToFortPawn HitActor))
  (bind _impactscale (Variables|Impact|GetImpactScale))
  ;[1 1 1] EVENT -> Collision Hit Effects:     Audio/FX Collision, used for skid fx and player hits
  (bind (_outresult _outmagnitude) (Audio|OnCollision (Variables|Default|GetFortCollisionAudioImpacts) HitLocation HitNormalImpulse))
  (CallFunction|PlayHapticsonHit _outmagnitude)
  (if (> _returnvalue 1e+04)
    (Audio|OnCollision (Variables|Default|GetFortCollisionAudioImpacts) HitLocation HitNormalImpulse)
    (else
      _))
  (if (Math|Boolean|ANDBoolean (> _returnvalue (Variables|Impact|GetScrapeImpact_Strength)) (>= (- (Utilities|Time|GetGameTimeinSeconds) (Variables|Impact|GetLastCollisionTime)) (Variables|Impact|GetMinTimeBetweenCollision)) (>= (- (Vehicle|GetTotalSpeedKmh) (Variables|FX|GetSpawnImpactFX_VehicleSpeed)) (Variables|FX|GetSpawnImpactFX_VehicleSpeed_Limit)))
    (if (Vehicle|VehicleIsinAir)
      ;[1 1 1] FX on land
      (bind (_outhits _returnvalue_4) (Collision|MultiLineTraceByChannel (+ (Transformation|GetActorLocation) "0,0,500.000000") (+ (Transformation|GetActorLocation) "0.000000,0.000000,-600.000000") "TraceTypeQuery10"))
      (if _returnvalue_4
        (if (== HitSurfaceType 30)
          (bind _returnvalue_5 (Effects|Components|ParticleSystem|SpawnFXAtLocation "/Game/Effects/MudInteraction/NS_Physical_Surface_Explosion.NS_Physical_Surface_Explosion" (.Location (Collision|BreakHitResult (Utilities|Array|Get(acopy) _outhits)))))
          (Effects|Components|ParticleSystem|SetFloatParameter _returnvalue_5 "Global Scale" 0.38)
          (else
            (bind _returnvalue_6 (Effects|Components|ParticleSystem|SpawnFXAtLocation (Variables|FX|GetFX_LandDust) (.Location (Collision|BreakHitResult (Utilities|Array|Get(acopy) _outhits)))))
            (Effects|Components|ParticleSystem|SetBoolParameter _returnvalue_6 "LandOnSnowOrIce" (or (Utilities|Enum|Equal(Enum) HitSurfaceType "SurfaceType24") (Utilities|Enum|Equal(Enum) HitSurfaceType "SurfaceType23"))))))
      (else
        (Variables|Impact|SetLastCollisionTime (Utilities|Time|GetGameTimeinSeconds)) ; Setting Last Collidion Time Lowers Impact Amount
        (if _bsuccess
          ;;[1 1 1] Player Pawn impact
          (bind _returnvalue_7 (Effects|Components|ParticleSystem|SpawnFXAtLocation (Variables|FX|GetNiagaraFX_PlayerImpact) HitLocation _returnvalue_3 _impactscale))
          (Effects|Components|ParticleSystem|SetFloatParameter _returnvalue_7 "ImpactStrength" _returnvalue)
          (else
            ;;[1 1 1] World Impact, set to auto relase
            (bind _returnvalue_8 (Effects|Components|ParticleSystem|SpawnFXAtLocation (Variables|FX|GetTemplateScrapeImpactFX) HitLocation _returnvalue_3 _impactscale true "AutoRelease"))
            (Effects|Components|ParticleSystem|SetFloatParameter _returnvalue_8 "ImpactStrength" _returnvalue)))))))

(event EventOnChargingJump
  (params (JumpChargeFactor :type Float%20%28single-precision%29))
  (bind _jumpchargefactor (Variables|Jump/Tricks|GetJumpChargeFactor))
  ;[1 1 1] Jump Events
  (Variables|Jump/Tricks|SetJumpChargeFactor JumpChargeFactor)
  (Effects|Components|ParticleSystem|SetFloatParameter (Variables|Default|GetNS_DirtBike_RoadAndExhaust) "ChargeProgress" _jumpchargefactor)
  (Variables|Camera|SetCameraFOVOffset (* _jumpchargefactor (Variables|Jump/Tricks|GetJumpFOVOffset))))

(event EventOnJump
  (params (JumpChargeFactor :type Float%20%28single-precision%29))
  (bind _ns_dirtbike_roadandexhaust (Variables|Default|GetNS_DirtBike_RoadAndExhaust))
  ;[1 1 1] Jump Events
  (Effects|Components|ParticleSystem|SetBoolParameter (Variables|Default|GetNS_DirtBike_RoadAndExhaust) "IsBoostJump" true)
  (bind _outhits (Collision|MultiLineTraceByChannel (+ (Transformation|GetActorLocation) "0,0,500.000000") (+ (Transformation|GetActorLocation) "0,0,-500.000000") "TraceTypeQuery10"))
  (if (== (Variables|/Script/PhysicsCore.PhysicalMaterial|GetSurfaceType (.PhysMat (Collision|BreakHitResult (Utilities|Array|Get(acopy) _outhits)))) 30)
    (bind _returnvalue (Effects|Components|ParticleSystem|SpawnFXAtLocation "/Game/Effects/MudInteraction/NS_Physical_Surface_Explosion.NS_Physical_Surface_Explosion" (+ (.Location (Collision|BreakHitResult (Utilities|Array|Get(acopy) _outhits))) "-50.000000,0.000000,0.000000")))
    (Effects|Components|ParticleSystem|SetFloatParameter _returnvalue "Global Scale" 0.4)
    (else
      (bind _returnvalue_2 (Effects|Components|ParticleSystem|SpawnFXAtLocation (Variables|FX|GetFX_JumpBoostSmokeKick) (+ (.Location (Collision|BreakHitResult (Utilities|Array|Get(acopy) _outhits))) "0,0,30.000000")))))
  (Effects|Components|ParticleSystem|SetBoolParameter _returnvalue_2 "JumpOnSnowOrIce" (or (Utilities|Enum|Equal(Enum) (Variables|/Script/PhysicsCore.PhysicalMaterial|GetSurfaceType (.PhysMat (Collision|BreakHitResult (Utilities|Array|Get(acopy) _outhits)))) "SurfaceType23") (Utilities|Enum|Equal(Enum) (Variables|/Script/PhysicsCore.PhysicalMaterial|GetSurfaceType (.PhysMat (Collision|BreakHitResult (Utilities|Array|Get(acopy) _outhits)))) "SurfaceType24")))
  (Variables|Camera|SetCameraFOVOffset (Variables|Jump/Tricks|GetDefault%20FOVOffset))
  (Effects|Components|ParticleSystem|SetBoolParameter _ns_dirtbike_roadandexhaust "SpawnExhaustFlames" true)
  (Utilities|FlowControl|RetriggerableDelay (Math|Float|Lerp 0.1 0.7 (Variables|Jump/Tricks|GetJumpChargeFactor)))
  (Effects|Components|ParticleSystem|SetBoolParameter _ns_dirtbike_roadandexhaust "SpawnExhaustFlames" false)
  (Effects|Components|ParticleSystem|SetFloatParameter _ns_dirtbike_roadandexhaust "ChargeProgress" 0.0)
  (Effects|Components|ParticleSystem|SetBoolParameter _ns_dirtbike_roadandexhaust "IsBoostJump" false))

(event EventOnCancelChargingJump
  ;[1 1 1] Jump Events
  (Variables|Camera|SetCameraFOVOffset (Variables|Jump/Tricks|GetDefault%20FOVOffset)))

(fn AddCooldownCueEvent
  (params (PlayerPawn :type Fort%20Player%20Pawn%20Object%20Reference)))

(fn AddToPlayerControllerArray
  (params (Target :type Pawn%20Object%20Reference))
  (bind _asplayer_controller (Utilities|Casting|CastToPlayerController (Pawn|GetController Target))
    (:then
      (Utilities|Array|Add (Variables|PawnReferences|GetValidPlayerControllers) _asplayer_controller))
    (:CastFailed
      _))
  (return))

(fn CacheExplosionDamageValues
  ;[1 0.1237 0.6737] Cache the explosion damage values from the default values (which are hooked up to data tables to be hotfixable). Cached values may be overridden by Creative Mode runtime user editable options.
  (Variables|Explosion|SetCachedExplosionDamageBuildings (ScalableFloat|GetValueatLevel (Variables|Hotfix|GetExplosionDamageBuildings)))
  (Variables|Explosion|SetCachedExplosionDamagePlayers (ScalableFloat|GetValueatLevel (Variables|Hotfix|GetExplosionDamagePlayers)))
  (Variables|Explosion|SetCachedExplosionDamageVehicles (ScalableFloat|GetValueatLevel (Variables|Hotfix|GetExplosionDamageVehicles))))

(fn CalcAccelRumbleIntensity
  (params (MaxIntensity :type Float%20%28double-precision%29))
  (return (Math|Float|Clamp(Float) (- 1.0 (Math|Float|Power (/ (Math|Float|Absolute(Float) (Vehicle|GetForwardSpeedKmh)) 1.0) (LocalVariables|GetAccelRumbleIntensityFalloff))) 0.0 MaxIntensity)))

(fn DamageablePartChanged
  (params (PartConfig :type Vehicle%20Damageable%20Part%20Config%20Structure) (Part :type Vehicle%20Damageable%20Part%20Structure) (NewlyDisabled :type Boolean))
  (bind (_shapename _bonename) (Utilities|Struct|BreakVehicleDamageablePartConfig PartConfig))
  (bind _ns_dirtbike_roadandexhaust (Variables|Default|GetNS_DirtBike_RoadAndExhaust))
  (if NewlyDisabled
    (LocalVariables|SetBoneNameLocal _bonename)
    (switch name _shapename
      (:wheel_f
        (Utilities|SpawnFXAttachedtoSocket (LocalVariables|GetWheelSocket_F) (LocalVariables|GetPoppedTireFX))
        (Vehicle|Tires|OnPoppedTire 0)
        (bind _output_get (Variables|States|SetTirePoppedFront true))
        (Niagara|SetNiagaraVariable(Bool) _ns_dirtbike_roadandexhaust "TirePoppedSparksFront" _output_get))
      (:wheel_b
        (Utilities|SpawnFXAttachedtoSocket (LocalVariables|GetWheelSocket_B) (LocalVariables|GetPoppedTireFX))
        (Vehicle|Tires|OnPoppedTire 1)
        (bind _output_get_2 (Variables|States|SetTirePoppedRear true))
        (Niagara|SetNiagaraVariable(Bool) _ns_dirtbike_roadandexhaust "TirePoppedSparksRear" _output_get_2))
      (:glass_fr_w
        (CallFunction|ToggleGlassFX (LocalVariables|GetBoneNameLocal) (LocalVariables|GetWindowTransform_Fr_W))))))

(fn GearShiftFovChange
  (params (NewGear :type Integer))
  (if (>= NewGear (Variables|GearShifting|GetCurrentGear))
    (Variables|Camera|SetCameraFOVOffset -3.0)
    (else
      (Variables|Camera|SetCameraFOVOffset -1.5)))
  (Variables|GearShifting|SetCurrentGear NewGear))

(fn GenerateRandomInt
  (params (TargetArray :type Array%20of%20Material%20Interface%20Object%20References))
  (bind _forcedmaterialvariantindex (Variables|FortAthenaVehicle|GetForcedMaterialVariantIndex))
  (if (== _forcedmaterialvariantindex -1)
    (Variables|RandomMaterial|SetRandomMatInt (Math|Random|RandomIntegerinRange 0 (- (Utilities|Array|Length TargetArray) 1)))
    (else
      (Variables|RandomMaterial|SetRandomMatInt _forcedmaterialvariantindex)))
  (return))

(fn HandleSurfaceTypeChanged
  (params (SurfaceType :type EPhysicalSurface%20Enum :default "SurfaceType_Default"))
  (if (Utilities|Enum|Equal(Enum) SurfaceType "SurfaceType24")
    (GameplayEffects|ApplyGameplayEffectToSelf :self (Vehicle|GetVehicleAbilitySystemComponent) :GameplayEffectClass "/Motorcycle/Abilities/GE_Motorcycle_Surface_Ice.GE_Motorcycle_Surface_Ice_C" :Level 0.0)
    (else
      (GameplayEffects|RemoveActiveGameplayEffectbySourceEffect (Vehicle|GetVehicleAbilitySystemComponent) "/Motorcycle/Abilities/GE_Motorcycle_Surface_Ice.GE_Motorcycle_Surface_Ice_C" 0 -1))))

(fn HandleRefueledFromEmpty
  (Niagara|SetNiagaraVariable(Bool) (Variables|Default|GetNS_DirtBike_RoadAndExhaust) "VehicleHasFuel" true))

(fn IsDriverRestrictedbyTeamSettings
  (params (PlayerPawn :type Fort%20Player%20Pawn%20Object%20Reference))
  ;[0.8734 0.03282 1] Is the driver restricted by team setting
  (return (CallFunction|IsDriverMinigameTeamRestricted PlayerPawn)))

(fn IsVehicleAccelerating?
  (params (MaxSpeed :type Float%20%28double-precision%29))
  (return (Math|Boolean|ANDBoolean (Math|Float|InRange(Float) (Vehicle|GetForwardSpeedKmh) 2.0 MaxSpeed) (> (Vehicle|GetForwardSpeedKmh) (Vehicle|GetPreviousForwardSpeedKmh)) (Vehicle|GasPedalIsDown) (>= (Vehicle|Fuel|GetFuel) 0.01) (not (Vehicle|VehicleIsinAir)))))

(fn IsVehicleOccupied
  (return (> (Utilities|Array|Length (Vehicle|GetAllPassengers)) 0)))

(fn OnRep_RandomMatInt
  (bind _sk_randommat2_array (Variables|RandomMaterial|GetSK_RandomMat2_Array))
  (bind _sk_randommat0_array (Variables|RandomMaterial|GetSK_RandomMat0_Array))
  (bind _sk_randommat1_array (Variables|RandomMaterial|GetSK_RandomMat1_Array))
  (if (> (Utilities|Array|Length _sk_randommat2_array) 0)
    (Class|FortVehicleInterface|ApplyMaterialandUseWrapifPossible (Utilities|Array|Get(acopy) _sk_randommat2_array (Variables|RandomMaterial|GetRandomMatInt)) 2)
    (else
      _))
  (if (> (Utilities|Array|Length _sk_randommat0_array) 0)
    (Class|FortVehicleInterface|ApplyMaterialandUseWrapifPossible (Utilities|Array|Get(acopy) _sk_randommat0_array (Variables|RandomMaterial|GetRandomMatInt)))
    (else
      _))
  (if (> (Utilities|Array|Length _sk_randommat1_array) 0)
    (Class|FortVehicleInterface|ApplyMaterialandUseWrapifPossible (Utilities|Array|Get(acopy) _sk_randommat1_array (Variables|RandomMaterial|GetRandomMatInt)) 1)))

(fn PhysicalSurfaceIDCheck
  (params (Physical%20Surface%20ID :type Integer))
  (return (Math|Boolean|ORBoolean (== Physical%20Surface%20ID 30) (== Physical%20Surface%20ID 26) (== Physical%20Surface%20ID 23) (== Physical%20Surface%20ID 24))))

(fn PickRandomMaterial
  (bind _sk_randommat0_array (Variables|RandomMaterial|GetSK_RandomMat0_Array))
  (if (Networking|IsServer)
    (if (> (Utilities|Array|Length _sk_randommat0_array) 0)
      (CallFunction|GenerateRandomInt _sk_randommat0_array)
      (else
        (CallFunction|GenerateRandomInt (select (> (Utilities|Array|Length (Variables|RandomMaterial|GetSK_RandomMat2_Array)) 0) (Variables|RandomMaterial|GetSK_RandomMat2_Array) (Variables|RandomMaterial|GetSK_RandomMat1_Array)))))))

(fn PlayHapticsOnHit
  (params (HitMagnitude :type Float%20%28double-precision%29))
  (bind _passengerplayercontroller (Variables|PawnReferences|GetPassengerPlayerController))
  (bind _drivingplayercontroller (Variables|PawnReferences|GetDrivingPlayerController))
  (bind _validplayercontrollers (Variables|PawnReferences|GetValidPlayerControllers))
  (Utilities|FlowControl|Sequence
    (:then_0
      (Utilities|IsValid _drivingplayercontroller
        (:Is%20Valid
          (Utilities|Array|Add (Variables|PawnReferences|GetValidPlayerControllers) _drivingplayercontroller))))
    (:then_1
      (Utilities|IsValid _passengerplayercontroller
        (:Is%20Valid
          (Utilities|Array|Add (Variables|PawnReferences|GetValidPlayerControllers) _passengerplayercontroller))))
    (:then_2
      (if (> (Utilities|Array|Length (Variables|PawnReferences|GetValidPlayerControllers)) 0)
        (if (> (Variables|GetHitMagnitude) 1e+04)
          (Utilities|FlowControl|Sequence
            (:then_0
              (LocalVariables|SetCameraShakeMagnitude (Math|Float|MapRangeClamped (Variables|GetHitMagnitude) 1.5e+04 5e+04 0.1 1.0))
              (Utilities|IsValid (Variables|PawnReferences|GetDrivingPlayerController)
                (:Is%20Valid
                  (Game|Feedback|ClientStartCameraShake :self (Variables|PawnReferences|GetValidPlayerControllers) :Shake (LocalVariables|GetSmallHitCameraShake) :Scale (LocalVariables|GetCameraShakeMagnitude) :PlaySpace "CameraLocal")
                  (if (> (Variables|GetHitMagnitude) 2e+04)
                    (Game|Feedback|ClientPlayForceFeedback _validplayercontrollers (LocalVariables|GetBigHitForceFeedback) "None" false false false)
                    (else
                      (Game|Feedback|ClientPlayForceFeedback _validplayercontrollers (LocalVariables|GetSmallHitForceFeedback) "None" false false false))))))
            (:then_1
              (if (< (Math|Float|Absolute(Float) (Math|Vector|DotProduct (Math|Vector|Normalize (Vehicle|GetPreviousLinearVelocity)) "0, 0,1.000000")) 0.5)
                (bind _output_get (LocalVariables|SetHitDelta (- (Variables|Vehicle|GetCachedSpeed) (Math|Vector|VectorLength (Transformation|GetComponentVelocity (Variables|Vehicle|GetSkeletalMesh))))))
                (if (> _output_get (LocalVariables|GetCameraPushVelocityDelta))
                  (Game|Feedback|ClientStartCameraShake (Variables|PawnReferences|GetValidPlayerControllers) (LocalVariables|GetBigHitCameraShake) (Math|Float|MapRangeClamped (LocalVariables|GetHitDelta) 3e+02 9e+02 0.5 1.0) "UserDefined" (Math|Rotator|MakeRotfromX (Vehicle|GetPreviousLinearVelocity))))))))))))

(fn RemoveCooldownCueEvent
  (params (PlayerPawn :type Fort%20Player%20Pawn%20Object%20Reference)))

(fn SendBoostTimeToMaterial
  (params (NewParam :type Float%20%28double-precision%29))
  (bind _output_get (Variables|Boost|SetSavedBoostTimeForMID NewParam))
  (Rendering|Material|SetScalarParameterValueonMaterials (Variables|Vehicle|GetSkeletalMesh) "BoostTime" _output_get))

(fn SendDamageValueToMaterials
  (params (NewParam :type Float%20%28double-precision%29))
  (bind _skeletalmesh (Variables|Vehicle|GetSkeletalMesh))
  (bind _healthpctmissingtoshowdmgcrinkle (LocalVariables|GetHealthPctMissingToShowDmgCrinkle))
  ;[1 1 1] Only Show damage crinkles after health is below a certain threshold
  (if (> NewParam _healthpctmissingtoshowdmgcrinkle)
    (bind _output_get (LocalVariables|SetSavedDamageValueForMID (Math|Float|Lerp (* (* _healthpctmissingtoshowdmgcrinkle (LocalVariables|GetMaxCrinkleSeverity)) -1.0) (LocalVariables|GetMaxCrinkleSeverity) NewParam)))
    (Rendering|Material|SetScalarParameterValueonMaterials _skeletalmesh "Damage" _output_get)
    (else
      (bind _output_get_2 (LocalVariables|SetSavedDamageValueForMID 0.0))
      (Rendering|Material|SetScalarParameterValueonMaterials _skeletalmesh "Damage" _output_get_2) ; Need to reset crinkle back to 0.0 if we're healed beyond the min threshold
    )))

(fn SetFallInstigator
  (params (InputPin :type Fort%20Player%20Pawn%20Athena%20Object%20Reference))
  (if (Vehicle|HasDriver)
    (Class|FortPlayerPawnAthena|SetFallInstigator :self InputPin :NewFallInstigator (Game|GetInstigatorController (Variables|PawnReferences|GetDriverPawn)))
    (elif (Variables|States|GetPassengerIsInstigator)
      (Class|FortPlayerPawnAthena|SetFallInstigator :self InputPin :NewFallInstigator (Game|GetInstigatorController (Utilities|Array|Get(acopy) (Variables|PawnReferences|GetPassengers))))
      (else
        (Class|FortPlayerPawnAthena|SetFallInstigator :self InputPin :NewFallInstigator (Game|GetInstigatorController (Variables|PawnReferences|GetLastDriver)))))))

(fn SetFxSettingsForVehicleType
  (bind _ns_dirtbike_roadandexhaust (Variables|Default|GetNS_DirtBike_RoadAndExhaust))
  (Effects|Components|ParticleSystem|SetFloatParameter _ns_dirtbike_roadandexhaust "Wheel_Width_Back" (LocalVariables|GetWheelWidthBack))
  (Niagara|SetNiagaraVariable(Bool) _ns_dirtbike_roadandexhaust "VehicleHasFuel" (> (Vehicle|Fuel|GetFuel) 0.0)))

(fn ShouldPlayNativeCurieFX
  (params (FXTag :type Gameplay%20Tag%20Structure))
  (return 0))

(fn SpawnFNRadialForce
  (params (SpawnTransform :type Transform) (Playrate :type Float%20%28double-precision%29) (TimeDecay :type Float%20%28double-precision%29) (ImpactRadius :type Float%20%28double-precision%29) (Strength :type Float%20%28double-precision%29))
  (Utilities|AllowJanus
    (:Allow
      (bind _returnvalue (Development|GetConsoleVariableIntValue "r.PhysicsField.EnableField"))
      (if (== _returnvalue 1)
        (bind _returnvalue_2 (Game|SpawnActorfromClass "/Game/Effects/Niagara/Blueprints/PhsyicsFields/FN_RadialForce.FN_RadialForce_C" SpawnTransform "Undefined" "MultiplyWithRoot" 0 Playrate TimeDecay ImpactRadius Strength 0.0 0))
        (LocalVariables|SetFN_RadialForce _returnvalue_2)
        (else
          _)))
    (:Nope
      _))
  (return))

(fn ToggleGlassFX
  (params (InputPin :type Name) (InTransform :type Transform))
  (Networking|SwitchHasAuthority
    (:Remote
      (if (Utilities|Name|NotEqual(Name) InputPin "")
        (bind _returnvalue (Niagara|SpawnSystemAttached "/Valet/FX/Niagara/NS_Glass_Blowout.NS_Glass_Blowout" (Variables|Vehicle|GetSkeletalMesh) InputPin (.location InTransform) (.rotation InTransform) "KeepRelativeOffset" false true "AutoRelease"))
        (Niagara|SetNiagaraVariable(Vector3) _returnvalue "windowSize" (.scale InTransform))))))

(fn ConstructionScript
  (Variables|FortAthenaVehicle|SetOverlapComponent (Variables|Default|GetOverlapVolume)))
